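# Unlit translucent sparkle material for Niagara sprite particles in the Vampire pickaxe effects.
# Particle color RGB passes through MF_FX_EmissiveAdjustments for the glow.
# Particle alpha multiplies the green channel of a four-point star sparkle texture to shape each sprite.
# Responsive AA is enabled for the translucent particles.
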Material /Game/Weapons/FORT_Melee/Effects/Vampire/M_Vampire_Pickaxe_Sparkles
graph {
 properties {
  bEnableResponsiveAA=true
  bUsedWithNiagaraSprites=true
  blendMode=BLEND_Translucent
  pixelDepthOffsetMode=PDOM_Legacy
  shadingModel=MSM_Unlit
  usedShadingModels=Unlit
 }
 nodes [
  MaterialFunctionCall(Emissive=ParticleColor().RGB) {materialFunction=&MF_FX_EmissiveAdjustments}
  Multiply(A=ParticleColor().A, B=TextureSample() {texture=/Game/Effects/Fort_Effects/Textures/Sparks/T_Sparkles_4Star_01_D}.G)
 ]
}
